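# Material function for the Rift Tour item shop UI that samples one flare shape from a bokeh channel texture.
# UVs are rotated, scaled around a point and shifted by the Rotation, Scale and Offset inputs, then sampled with clamping.
# The aspect factor divides pixel-size G by itself, and Shape blends red with red, so neither alters the result.
# A multiply by the scaling mask is left unconnected to the output.
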MaterialFunction /ItemShop/UI/Art/RiftTour/MF_UI_RiftTourShop_Flare
graph {nodes [
  n1 = Constant() {r=0.2}
  n2 = MaterialFunctionCall(Float2=MaterialFunctionCall() {materialFunction=&GetUserInterfaceUV}.Pixel Size) {materialFunction=&BreakOutFloat2Components}
  n3 = Divide(A=@n2.G, B=@n2.G)
  n4 = Constant() {r=1}
  n5 = Constant()
  n6 = MaterialFunctionCall!(
   Texture Scale=Multiply(A=FunctionInput(Preview=@n1) {bUsePreviewValueAsDefault=true, inputName=Scale, inputType=FunctionInput_Scalar}, B=AppendVector(A=@n3, B=@n4))
   UVs=MaterialFunctionCall(
    UVs=TextureCoordinate()
    "Rotation Angle (0-1)"=FunctionInput(Preview=@n5) {bUsePreviewValueAsDefault=true, inputName=Rotation, inputType=FunctionInput_Scalar}
   ) {materialFunction=&CustomRotator}.Rotated Values
   Offset XY Texturesize Z
  ) {materialFunction=/Engine/Functions/Engine_MaterialFunctions01/Texturing/ScaleUVsAroundPoint}
  n7 = Constant2Vector()
  n8 = TextureSample(UVs=Add(A=@n6.UVs, B=FunctionInput(Preview=@n7) {bUsePreviewValueAsDefault=true, inputName=Offset, inputType=FunctionInput_Vector2})) {samplerSource=SSM_Clamp_WorldGroupSettings, texture=&/T_UI_RiftShop_BokehChannels}
  n9 = LinearInterpolate(A=@n8.R, B=@n8.R, Alpha=FunctionInput(Preview=Constant()) {bUsePreviewValueAsDefault=true, inputName=Shape, inputType=FunctionInput_Scalar})
  FunctionOutput(None=@n9)
  Multiply(A=@n6.0-1 mask, B=@n9)
 ]}
