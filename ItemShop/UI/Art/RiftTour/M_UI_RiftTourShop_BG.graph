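# UI material for the animated Rift Tour Item Shop background, layering a curve-atlas gradient, two corner glows, stars, haze and flares.
# Stars come from a rotating sky mask multiplied by a panning copy for flicker; haze is flow-distorted, soft-masked noise sampled at offset UVs per channel for color separation.
# Two sets of four flares use a shared flare function, looped by curve-atlas motion, with the second set time-offset and rotated.
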Material /ItemShop/UI/Art/RiftTour/M_UI_RiftTourShop_BG
graph {
 properties {bUsedWithStaticMesh=false, materialDomain=MD_UI, numCustomizedUVs=2, pixelDepthOffsetMode=PDOM_Legacy, usedShadingModels=Default Lit}
 nodes [
  n1 = MaterialFunctionCall(UVs=NamedRerouteUsage(), Texture Scale=ScalarParameter() {defaultValue=1.5, group=Gradient, parameterName=GradientScale}) {materialFunction=&ScaleUVsByCenter}
  n2 = ScalarParameter() {defaultValue=0.3726, group=Gradient, parameterName=GradientAngle}
  n3 = CurveAtlasRowParameter(CurveTime=ComponentMask(None=MaterialFunctionCall(UVs=@n1.UVs, "Rotation Angle (0-1)"=@n2) {materialFunction=&CustomRotator}.Rotated Values) {r=true}) {
   atlas=/Game/UI/Frontend/ItemShop/Art/ItemShopColorAtlas
   curve=/Game/UI/Frontend/ItemShop/Art/ItemShop_RiftTourBG_Curve
   group=Gradient
   parameterName=CurveParamsBG
  }
  n4 = VectorParameter() {group=CornerGlows, parameterName=BottomRightGlow, defaultValue=<rgba 0.0782 0.855 1 1>}
  n5 = VectorParameter() {group=CornerGlows, parameterName=BottomRightGlow_Position, defaultValue=<rgba 0.7524 1.3528 1.3026 1>}
  n6 = MaterialFunctionCall(UVs=NamedRerouteUsage(), CenterPosition=AppendVector(A=@n5.R, B=@n5.R), Radius=@n5.B, Density=@n5.B) {materialFunction=&RadialGradientExponential}
  n7 = NamedRerouteUsage()
  n8 = VectorParameter() {group=CornerGlows, parameterName=TopLeftGlow_Position, defaultValue=<rgba -0.2571 -0.219 0.9381 1.9708>}
  n9 = AppendVector(A=@n8.R, B=@n8.R)
  n10 = Add(
   A=LinearInterpolate(
    A=LinearInterpolate(A=@n3, B=@n4.RGB, Alpha=@n6.RadialGradientExponential)
    B=VectorParameter() {group=CornerGlows, parameterName=TopLeftGlow, defaultValue=<rgba 1 0.7084 0.2423 1>}.RGB
    Alpha=MaterialFunctionCall(UVs=@n7, CenterPosition=@n9, Radius=@n8.B, Density=@n8.B) {materialFunction=&RadialGradientExponential}.RadialGradientExponential
   )
   B=NamedRerouteUsage()
  )
  n11 = Add(A=@n10, B=Power(Base=NamedRerouteUsage(), Exp=ScalarParameter() {defaultValue=1, group=Haze, parameterName=HazePower}))
  n12 = ScalarParameter() {defaultValue=0.7, group=Haze, parameterName=HazeAmount}
  Add(A=Add(A=LinearInterpolate(A=@n10, B=@n11, Alpha=@n12), B=NamedRerouteUsage()) {constB=0}, B=NamedRerouteUsage())
  n13 = VectorParameter() {parameterName=Flare3, channelNames {a=Shape, b=Scale, g=Ystart, r=XStart}, defaultValue=<rgba -1 -0.4 0.3 1>}
  n14 = Divide(B=@n13.Scale)
  n15 = Add(A=Time(), B=ScalarParameter() {defaultValue=0.8, parameterName=Flare3TimeOffset})
  n16 = ScalarParameter() {defaultValue=0.2, parameterName=Flare3MotionSpeed}
  n17 = CurveAtlasRowParameter(CurveTime=Fmod(A=Multiply(A=@n15, B=@n16), B=Constant() {r=1})) {
   atlas=&ItemShopColorAtlas
   curve=/Game/UI/Frontend/ItemShop/Art/ItemShop_RiftTourFlare_Curve
   defaultValue=1
   group=Gradient
   parameterName=FlareMotion
  }
  n18 = Add(A=Multiply(A=@n14, B=AppendVector(A=@n13.XStart, B=@n13.XStart)), B=Multiply(A=@n14, B=AppendVector(A=@n17.R, B=@n17.R)))
  n19 = Multiply(A=@n13.Scale, B=@n17.B)
  n20 = Multiply(
   A=Multiply(
    A=Constant3Vector() {constant=<rgba 0.6621 0 1 1>}
    B=MaterialFunctionCall!(Offset=@n18, Scale=@n19, Rotation, Shape=@n13.Shape) {materialFunction=&/MF_UI_RiftTourShop_Flare}.Result
   )
   B=@n17.A
  )
  n21 = VectorParameter() {group=Bokeh1, parameterName=Flare1Color, defaultValue=<rgba 0 0.8057 1 1>}
  n22 = VectorParameter() {parameterName=Flare1, channelNames {a=Shape, b=Scale, g=Ystart, r=XStart}, defaultValue=<rgba -0.9476 -0.6 0.1905 0>}
  n23 = Divide(B=@n22.Scale)
  n24 = AppendVector(A=@n22.XStart, B=@n22.XStart)
  n25 = Time()
  n26 = ScalarParameter() {defaultValue=0.1, parameterName=FlareMotionSpeed}
  n27 = CurveAtlasRowParameter(CurveTime=Fmod(A=Multiply(A=@n25, B=@n26), B=Constant() {r=1})) {
   atlas=&ItemShopColorAtlas
   curve=&ItemShop_RiftTourFlare_Curve
   defaultValue=1
   group=Gradient
   parameterName=FlareMotion
  }
  n28 = AppendVector(A=@n27.R, B=@n27.R)
  n29 = MaterialFunctionCall(Offset=Add(A=Multiply(A=@n23, B=@n24), B=Multiply(A=@n23, B=@n28)), Scale=Multiply(A=@n22.Scale, B=@n27.B), Shape=@n22.Shape) {materialFunction=&MF_UI_RiftTourShop_Flare}
  n30 = VectorParameter() {group=Bokeh1, parameterName=FlareColor2, defaultValue=<rgba 0.1768 0.3409 0.355 1>}
  n31 = VectorParameter() {parameterName=Flare2, channelNames {a=Shape, b=Scale, g=Ystart, r=XStart}, defaultValue=<rgba -1.2476 -0.6 0.3 0>}
  n32 = Add(A=Time(), B=ScalarParameter() {defaultValue=0.4, parameterName=Flare2TimeOffset})
  n33 = ScalarParameter() {defaultValue=0.25, parameterName=Flare2MotionSpeed}
  n34 = CurveAtlasRowParameter(CurveTime=Fmod(A=Multiply(A=@n32, B=@n33), B=Constant() {r=1})) {
   atlas=&ItemShopColorAtlas
   curve=&ItemShop_RiftTourFlare_Curve
   defaultValue=1
   group=Gradient
   parameterName=FlareMotion
  }
  n35 = Divide(B=@n31.Scale)
  n36 = AppendVector(A=@n31.XStart, B=@n31.XStart)
  n37 = AppendVector(A=@n34.R, B=@n34.R)
  n38 = MaterialFunctionCall(Offset=Add(A=Multiply(A=@n35, B=@n36), B=Multiply(A=@n35, B=@n37)), Scale=Multiply(A=@n31.Scale, B=@n34.B), Shape=@n31.Shape) {materialFunction=&MF_UI_RiftTourShop_Flare}
  n39 = Add(A=Multiply(A=Multiply(A=@n21.RGB, B=@n29.Result), B=@n27.A), B=Multiply(A=Multiply(A=@n30.RGB, B=@n38.Result), B=@n34.A))
  n40 = VectorParameter() {parameterName=Flare4, channelNames {a=Shape, b=Scale, g=Ystart, r=XStart}, defaultValue=<rgba -1.2476 -0.6 0.3 0>}
  n41 = Divide(B=@n40.Scale)
  n42 = Multiply(A=@n41, B=AppendVector(A=@n40.XStart, B=@n40.XStart))
  n43 = Add(A=Time(), B=ScalarParameter() {defaultValue=0.7, parameterName=Flare4TimeOffset})
  n44 = ScalarParameter() {defaultValue=0.25, parameterName=Flare4MotionSpeed}
  n45 = CurveAtlasRowParameter(CurveTime=Fmod(A=Multiply(A=@n43, B=@n44), B=Constant() {r=1})) {
   atlas=&ItemShopColorAtlas
   curve=&ItemShop_RiftTourFlare_Curve
   defaultValue=1
   group=Gradient
   parameterName=FlareMotion
  }
  n46 = Multiply(A=@n41, B=AppendVector(A=@n45.R, B=@n45.R))
  n47 = Multiply(
   A=Constant3Vector() {constant=<rgba 0.1768 0.3409 0.355 1>}
   B=MaterialFunctionCall(Offset=Add(A=@n42, B=@n46), Scale=Multiply(A=@n40.Scale, B=@n45.B), Shape=@n40.Shape) {materialFunction=&MF_UI_RiftTourShop_Flare}.Result
  )
  n48 = Multiply(A=@n45.A) {constB=2}
  NamedRerouteDeclaration(None=Add(A=Add(A=@n20, B=@n39), B=Multiply(A=@n47, B=@n48))) {name=Flares, nodeColor=<rgba 0.2656 0 1 1>}
  n49 = TextureCoordinate()
  n50 = MaterialFunctionCall(Float2=MaterialFunctionCall() {materialFunction=&GetUserInterfaceUV}.Pixel Size) {materialFunction=&BreakOutFloat2Components}
  n51 = AppendVector(A=Constant() {r=1}, B=Divide(A=@n50.R, B=@n50.R))
  n52 = MaterialFunctionCall(
   UVs=NamedRerouteDeclaration(None=MaterialFunctionCall(UVs=@n49, Texture Scale=@n51) {materialFunction=&ScaleUVsByCenter}.UVs) {name=ScreenNormalizedUVs, nodeColor=<rgba 1 0.0938 0 1>}
   Texture Scale=ScalarParameter() {defaultValue=1, group=Stars, parameterName=StarScale}
  ) {materialFunction=&ScaleUVsByCenter}
  n53 = ScalarParameter() {group=Stars, parameterName=StarPan}
  n54 = VectorParameter() {group=Stars, parameterName=StarRotation, defaultValue=<rgba 0 0.581 0.0025 0>}
  n55 = Multiply(A=Time(), B=@n54.B)
  n56 = Constant() {r=1}
  n57 = TextureSample(UVs=MaterialFunctionCall(UVs=Panner(Coordinate=@n52.UVs, Speed=@n53), Rotation Center=AppendVector(A=@n54.R, B=@n54.R), "Rotation Angle (0-1)"=Fmod(A=@n55, B=@n56)) {materialFunction=&CustomRotator}.Rotated Values) {constMipValue=1, samplerType=SAMPLERTYPE_LinearColor, texture=/Game/Packages/Fortress_Sky/SkyDome/Master/T_SkyDome_StarCloudMask}
  n58 = Panner(Coordinate=NamedRerouteUsage(), Speed=ScalarParameter() {defaultValue=0.025, group=Stars, parameterName=StarFlickerSpeed})
  n59 = ShadingPathSwitch(Default=Multiply(
    A=ComponentMask(None=@n57.R) {r=true}
    B=TextureSample(UVs=@n58) {samplerType=SAMPLERTYPE_LinearColor, texture=&T_SkyDome_StarCloudMask}.G
   ))
  n60 = ScalarParameter() {defaultValue=2, group=Stars, parameterName=StarPower}
  NamedRerouteDeclaration(None=Multiply(A=Power(Base=@n59, Exp=@n60), B=Constant() {r=100})) {name=Stars_1, nodeColor=<rgba 0.8414 0 1 1>}
  n61 = ScalarParameter() {defaultValue=0.5, parameterName=TopFlareSetRotation}
  n62 = VectorParameter() {parameterName=Flare3, channelNames {a=Shape, b=Scale, g=Ystart, r=XStart}, defaultValue=<rgba -1 -0.4 0.3 1>}
  n63 = Divide(B=@n62.Scale)
  n64 = Add(
   A=Add(A=Time(), B=ScalarParameter() {defaultValue=0.8, parameterName=Flare3TimeOffset})
   B=ScalarParameter() {defaultValue=0.8, parameterName=FlareSet2Offset}
  )
  n65 = ScalarParameter() {defaultValue=0.2, parameterName=Flare3MotionSpeed}
  n66 = CurveAtlasRowParameter(CurveTime=Fmod(A=Multiply(A=@n64, B=@n65), B=Constant() {r=1})) {
   atlas=&ItemShopColorAtlas
   curve=&ItemShop_RiftTourFlare_Curve
   defaultValue=1
   group=Gradient
   parameterName=FlareMotion
  }
  n67 = Add(A=Multiply(A=@n63, B=AppendVector(A=@n62.XStart, B=@n62.XStart)), B=Multiply(A=@n63, B=AppendVector(A=@n66.R, B=@n66.R)))
  n68 = Multiply(A=@n62.Scale, B=@n66.B)
  n69 = Multiply(
   A=Multiply(
    A=Constant3Vector() {constant=<rgba 0 1 0.434 1>}
    B=MaterialFunctionCall(Offset=@n67, Scale=@n68, Rotation=@n61, Shape=@n62.Shape) {materialFunction=&MF_UI_RiftTourShop_Flare}.Result
   )
   B=@n66.A
  )
  n70 = VectorParameter() {group=Bokeh1, parameterName=Flare1Color, defaultValue=<rgba 0 0.8057 1 1>}
  n71 = VectorParameter() {parameterName=Flare1, channelNames {a=Shape, b=Scale, g=Ystart, r=XStart}, defaultValue=<rgba -0.9476 -0.6 0.1905 0>}
  n72 = Divide(B=@n71.Scale)
  n73 = AppendVector(A=@n71.XStart, B=@n71.XStart)
  n74 = Add(A=Time(), B=ScalarParameter() {defaultValue=0.8, parameterName=FlareSet2Offset})
  n75 = ScalarParameter() {defaultValue=0.1, parameterName=FlareMotionSpeed}
  n76 = CurveAtlasRowParameter(CurveTime=Fmod(A=Multiply(A=@n74, B=@n75), B=Constant() {r=1})) {
   atlas=&ItemShopColorAtlas
   curve=&ItemShop_RiftTourFlare_Curve
   defaultValue=1
   group=Gradient
   parameterName=FlareMotion
  }
  n77 = AppendVector(A=@n76.R, B=@n76.R)
  n78 = MaterialFunctionCall(
   Offset=Add(A=Multiply(A=@n72, B=@n73), B=Multiply(A=@n72, B=@n77))
   Scale=Multiply(A=@n71.Scale, B=@n76.B)
   Rotation=ScalarParameter() {defaultValue=0.5, parameterName=TopFlareSetRotation}
   Shape=@n71.Shape
  ) {materialFunction=&MF_UI_RiftTourShop_Flare}
  n79 = VectorParameter() {group=Bokeh1, parameterName=FlareColor2, defaultValue=<rgba 0.1768 0.3409 0.355 1>}
  n80 = VectorParameter() {parameterName=Flare2, channelNames {a=Shape, b=Scale, g=Ystart, r=XStart}, defaultValue=<rgba -1.2476 -0.6 0.3 0>}
  n81 = Divide(B=@n80.Scale)
  n82 = AppendVector(A=@n80.XStart, B=@n80.XStart)
  n83 = Add(
   A=Add(A=Time(), B=ScalarParameter() {defaultValue=0.4, parameterName=Flare2TimeOffset})
   B=ScalarParameter() {defaultValue=0.8, parameterName=FlareSet2Offset}
  )
  n84 = ScalarParameter() {defaultValue=0.25, parameterName=Flare2MotionSpeed}
  n85 = CurveAtlasRowParameter(CurveTime=Fmod(A=Multiply(A=@n83, B=@n84), B=Constant() {r=1})) {
   atlas=&ItemShopColorAtlas
   curve=&ItemShop_RiftTourFlare_Curve
   defaultValue=1
   group=Gradient
   parameterName=FlareMotion
  }
  n86 = AppendVector(A=@n85.R, B=@n85.R)
  n87 = MaterialFunctionCall(
   Offset=Add(A=Multiply(A=@n81, B=@n82), B=Multiply(A=@n81, B=@n86))
   Scale=Multiply(A=@n80.Scale, B=@n85.B)
   Rotation=ScalarParameter() {defaultValue=0.5, parameterName=TopFlareSetRotation}
   Shape=@n80.Shape
  ) {materialFunction=&MF_UI_RiftTourShop_Flare}
  n88 = Add(A=Multiply(A=Multiply(A=@n70.RGB, B=@n78.Result), B=@n76.A), B=Multiply(A=Multiply(A=@n79.RGB, B=@n87.Result), B=@n85.A))
  n89 = VectorParameter() {parameterName=Flare4, channelNames {a=Shape, b=Scale, g=Ystart, r=XStart}, defaultValue=<rgba -1.2476 -0.6 0.3 0>}
  n90 = Divide(B=@n89.Scale)
  n91 = Multiply(A=@n90, B=AppendVector(A=@n89.XStart, B=@n89.XStart))
  n92 = Add(
   A=Add(A=Time(), B=ScalarParameter() {defaultValue=0.7, parameterName=Flare4TimeOffset})
   B=ScalarParameter() {defaultValue=0.8, parameterName=FlareSet2Offset}
  )
  n93 = ScalarParameter() {defaultValue=0.25, parameterName=Flare4MotionSpeed}
  n94 = CurveAtlasRowParameter(CurveTime=Fmod(A=Multiply(A=@n92, B=@n93), B=Constant() {r=1})) {
   atlas=&ItemShopColorAtlas
   curve=&ItemShop_RiftTourFlare_Curve
   defaultValue=1
   group=Gradient
   parameterName=FlareMotion
  }
  n95 = Multiply(A=@n90, B=AppendVector(A=@n94.R, B=@n94.R))
  n96 = Multiply(
   A=Constant3Vector() {constant=<rgba 0.1768 0.3409 0.355 1>}
   B=MaterialFunctionCall(
    Offset=Add(A=@n91, B=@n95)
    Scale=Multiply(A=@n89.Scale, B=@n94.B)
    Rotation=ScalarParameter() {defaultValue=0.5, parameterName=TopFlareSetRotation}
    Shape=@n89.Shape
   ) {materialFunction=&MF_UI_RiftTourShop_Flare}.Result
  )
  n97 = Multiply(A=@n94.A) {constB=2}
  NamedRerouteDeclaration(None=Add(A=Add(A=@n69, B=@n88), B=Multiply(A=@n96, B=@n97))) {name=Flares2, nodeColor=<rgba 0 1 0.5078 1>}
  n98 = Add(A=VertexColor().R, B=TextureCoordinate() {coordinateIndex=1})
  n99 = ScalarParameter() {defaultValue=0.4, group=Haze, parameterName=TileSize_1}
  n100 = Time()
  n101 = ScalarParameter() {defaultValue=3, group=Haze, parameterName=Panning Scale}
  n102 = ScalarParameter() {defaultValue=1, group=Haze, parameterName=Velocity_Ctrl}
  n103 = Multiply(A=Constant2Vector() {g=-0.028}, B=@n102)
  n104 = TextureSampleParameter2D(UVs=Panner(Coordinate=Multiply(A=@n98, B=@n99), Time=Multiply(A=@n100, B=@n101), Speed=@n103) {speedY=-0.028}) {
   desc=Distort the Tiling Mask by this
   group=Haze
   parameterName=UVDistort_Normal
   samplerType=SAMPLERTYPE_Normal
   texture=/Game/Global/Textures/Default/Flowmaps/T_SwirlFlow_02_N
  }
  n105 = Multiply(A=ComponentMask(None=@n104.RGB) {g=true, r=true}, B=ScalarParameter() {defaultValue=0.08, group=Haze, parameterName=UVDistort})
  n106 = Add(B=TextureCoordinate() {coordinateIndex=1, uTiling=0.5, vTiling=0.5}) {constA=0.1}
  n107 = Multiply(A=Time(), B=ScalarParameter() {defaultValue=3, group=Haze, parameterName=Panning Scale})
  n108 = Multiply(A=Constant2Vector() {g=-0.0125}, B=@n102)
  n109 = TextureSampleParameter2D(UVs=Panner(Coordinate=Add(A=@n105, B=@n106), Time=@n107, Speed=@n108) {speedY=-0.0125}) {
   desc=UV1 Tiling Noise Mask
   group=Haze
   parameterName=TileAlphaMask
   samplerType=SAMPLERTYPE_LinearGrayscale
   texture=/Game/Global/Textures/Default/Noise/T_SwirlNoise_N_height
  }
  n110 = Add(B=TextureCoordinate() {coordinateIndex=1, uTiling=0.5, vTiling=0.5}) {constA=-0.1}
  n111 = TextureSampleParameter2D(UVs=Panner(Coordinate=Add(A=@n105, B=@n110), Time=@n107, Speed=@n108) {speedY=-0.0125}) {
   desc=UV1 Tiling Noise Mask
   group=Haze
   parameterName=TileAlphaMask
   samplerType=SAMPLERTYPE_LinearGrayscale
   texture=&T_SwirlNoise_N_height
  }
  n112 = TextureCoordinate() {coordinateIndex=1, uTiling=0.5, vTiling=0.5}
  n113 = TextureSampleParameter2D(UVs=Panner(Coordinate=Add(A=@n105, B=@n112), Time=@n107, Speed=@n108) {speedY=-0.0125}) {
   desc=UV1 Tiling Noise Mask
   group=Haze
   parameterName=TileAlphaMask
   samplerType=SAMPLERTYPE_LinearGrayscale
   texture=&T_SwirlNoise_N_height
  }
  n114 = AppendVector(A=AppendVector(A=@n109.R, B=@n111.R), B=@n113.B)
  n115 = Add(A=VertexColor().R, B=TextureCoordinate() {coordinateIndex=1})
  n116 = ScalarParameter() {defaultValue=0.1, group=Haze, parameterName=TileSize}
  n117 = TextureSampleParameter2D(UVs=Panner(Coordinate=Multiply(A=@n115, B=@n116), Time=@n107, Speed=@n103) {speedY=-0.028}) {
   desc=Distort the Tiling Mask by this
   group=Haze
   parameterName=UVDistort_Normal
   samplerType=SAMPLERTYPE_Normal
   texture=&T_SwirlFlow_02_N
  }
  n118 = Multiply(A=ComponentMask(None=@n117.RGB) {g=true, r=true}, B=ScalarParameter() {defaultValue=0.08, group=Haze, parameterName=UVDistort})
  n119 = TextureCoordinate() {coordinateIndex=1, uTiling=0.5, vTiling=0.5}
  n120 = TextureSampleParameter2D(UVs=Panner(Coordinate=Add(A=@n118, B=@n119), Time=@n107, Speed=@n108) {speedY=-0.0125}) {
   desc=UV1 Tiling Noise Mask
   group=Haze
   parameterName=TileAlphaMask
   samplerType=SAMPLERTYPE_LinearGrayscale
   texture=&T_SwirlNoise_N_height
  }
  n121 = MaterialFunctionCall(UVs=Add(A=TextureCoordinate(), B=@n118), Texture Scale=ScalarParameter() {defaultValue=1, parameterName=MaskScale}) {materialFunction=&ScaleUVsByCenter}
  n122 = Clamp(None=Multiply(
    A=@n120.G
    B=TextureSampleParameter2D(UVs=@n121.UVs) {desc=UV0 AlphaMask, group=Haze, parameterName=AlphaMask, texture=/Game/Effects/Fort_Effects/Textures/T_Smoke_Soft}.G
   ))
  NamedRerouteDeclaration(None=Saturate(None=Multiply(A=@n114, B=@n122))) {name=Haze, nodeColor=<rgba 1 0 0.375 1>}
  NamedRerouteUsage()
  TextureCoordinate()
 ]
}
